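; Pawn component that drives the WinterDecay void teleport, tracking the weapon's projectile and the teleport marker it spawns.
; Weapon, projectile and marker events chain through delegates, clearing old markers on new shots or death, while a looping timer updates marker position.
; Scalable floats toggle stopping the projectile, using its last location and teleporting while carrying DBNO players via async tag and input waits.
; Teleport completion fires GameplayCues and a client-side audio control-bus-mix failsafe.
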(event Custom|TriggerTeleportMarkerCue
  (Utilities|FlowControl|Sequence
    (:then_0
      (GameplayCue|ExecuteGameplayCue (Components|GetOwner) (Variables|Default|GetTeleport%20Marker%20Cue%20Tag)))
    (:then_1
      ;[0.0241 0 0.974 0.5] This is to update directional FX gcnl for moving position for now
      (bind _returnvalue (Utilities|Time|SetTimerbyEvent (EventDispatchers|CreateEvent :!event Custom|UpdateMarkerPosition) 1.0 true true))
      (Variables|Default|SetSet%20Marker%20Pos%20Timer%20Handle _returnvalue))))

(event EventBeginPlay
  (bind _ownerfortpawn (Variables|FortPawnComponent_VoidTeleport|GetOwnerFortPawn))
  (|CallParentFunction)
  (Utilities|IsValid _ownerfortpawn
    (:Is%20Valid
      (FortPawn|BindEventtoOnWeaponEquippedDelegate _ownerfortpawn (EventDispatchers|CreateEvent :!event NewWeaponEquipped))
      (EventDispatchers|BindEventtoOnDied _ownerfortpawn (EventDispatchers|CreateEvent :!event Custom|EventOnDied)))))

(event EventEndPlay
  (params (EndPlayReason :type EEndPlayReason%20Enum :default "Destroyed"))
  (bind _cbm_teleport (Variables|Default|GetCBM_Teleport))
  (bind _ownerfortpawn (Variables|FortPawnComponent_VoidTeleport|GetOwnerFortPawn))
  (|CallParentFunction EndPlayReason)
  (Utilities|IsValid _ownerfortpawn
    (:Is%20Valid
      (FortPawn|UnbindEventfromOnWeaponEquippedDelegate _ownerfortpawn (EventDispatchers|CreateEvent :!event NewWeaponEquipped))
      (Utilities|IsValid (Variables|Default|GetWinterDecayWeaponReference)
        (:Is%20Valid
          (Default|UnbindallEventsfromOnNewProjectileCreatedandValid (Variables|Default|GetWinterDecayWeaponReference)))
        (:Is%20Not%20Valid
          _))
      (Utilities|Time|ClearandInvalidateTimerbyHandle (Variables|Default|GetCBM_Handle))
      (EventDispatchers|UnbindEventfromOnDied _ownerfortpawn (EventDispatchers|CreateEvent :!event Custom|EventOnDied))
      (if (Audio|Modulation|IsControlBusMixActive _cbm_teleport)
        (Audio|Modulation|DeactivateControlBusMix _cbm_teleport)))
    (:Is%20Not%20Valid
      _)))

(event Custom|UpdateMarkerPosition
  (bind _bgatoteleport (Variables|FortPawnComponent_VoidTeleport|GetBGAToTeleport))
  (bind _winterdecayweaponreference (Variables|Default|GetWinterDecayWeaponReference))
  ;[0.0241 0 0.974 0.5] This is to update directional FX gcnl for moving position for now
  (Utilities|IsValid _bgatoteleport
    (:Is%20Valid
      (Utilities|IsValid _winterdecayweaponreference
        (:Is%20Valid
          (Variables|/WinterDecay/Gameplay/Actors/B_WinterDecay_Teleport.B_WinterDecay_Teleport_C|SetPositionToTeleport (Transformation|GetActorLocation _bgatoteleport) _winterdecayweaponreference))))))

(event EventForceCrouch
  (|CallParentFunction))

(event Custom|TryTeleportWhileCarryingDBNOPlayer%20Thread
  ;[0.3265 0 0.9948 0.5] Teleport carrying DBNO players
  (bind _asynctaskproxy (Ability|Tasks|WaitGameplayTagAddToActor (Variables|FortPawnComponent_VoidTeleport|GetOwnerFortPawn) (Variables|Default|GetCarrying%20Player%20Tag))
    (:then
      (Variables|Default|SetCarrying%20Player%20Async%20Action _asynctaskproxy))
    (:Added
      (Variables|FortPawnComponent_VoidTeleport|GetBGAToTeleport
        (:then
          (Utilities|Casting|CastToB_DemonMask_Void_Teleport (Variables|/Script/FortniteGame.FortPawn|GetCurrentWeapon (Variables|FortPawnComponent_VoidTeleport|GetOwnerFortPawn))
            (:CastFailed
              (bind _asynctaskproxy_2 (Ability|Async|WaitGameplayEventToActor (Variables|FortPawnComponent_VoidTeleport|GetOwnerFortPawn) "(TagName=\"Event.Secondary.Pressed\")")
                (:then
                  (Variables|Default|SetTeleport%20Input%20Pressed%20Async%20Action _asynctaskproxy_2)
                  (bind _asynctaskproxy_3 (Ability|Tasks|WaitGameplayTagRemoveFromActor (Variables|FortPawnComponent_VoidTeleport|GetOwnerFortPawn) (Variables|Default|GetCarrying%20Player%20Tag))
                    (:then
                      (Variables|Default|SetCarrying%20Tag%20Removed%20Async%20Action _asynctaskproxy_3))
                    (:Removed
                      (Ability|Async|EndAction (Variables|Default|GetTeleport%20Input%20Pressed%20Async%20Action))
                      (Utilities|FlowControl|Delay 0.2)
                      (CallFunction|TryTeleportWhileCarryingDBNOPlayerThread))))
                (:EventReceived
                  (CallFunction|ServerAddGameplayTag))))))
        (:else)))))

(event Custom|PlayerComponentStartTeleport
  ;[0 0.801 0.8802 0.5] Comment
  (Utilities|FlowControl|Sequence
    (:then_0
      (Networking|SwitchHasAuthority))))

(event Custom|PlayerComponentEndTeleport
  ;[0 0.801 0.8802 0.5] Comment
  (Utilities|FlowControl|Sequence
    (:then_0
      (Networking|SwitchHasAuthority
        (:Remote
          (bind _returnvalue (Utilities|Time|SetTimerbyEvent (delegate Custom|DemonMaskVoid_DeactivateControlBusMixFailsafe) 0.2 true))
          (Variables|Default|SetCBM_Handle _returnvalue))))))

(event Custom|Server_AddGameplayTag
  ;[0.3265 0 0.9948 0.5] Teleport carrying DBNO players
  (Ability|SendGameplayEventtoActor (Variables|FortPawnComponent_VoidTeleport|GetOwnerFortPawn) (Variables|Default|GetDBNO%20Teleport%20Event%20Tag) (Utilities|Struct|MakeGameplayEventData "(TagName=\"\")")))

(event Custom|DemonMaskVoid_DeactivateControlBusMixFailsafe
  (bind _cbm_teleport (Variables|Default|GetCBM_Teleport))
  ;[0 0.801 0.8802 0.5] Comment
  (if (Audio|Modulation|IsControlBusMixActive _cbm_teleport)
    (Audio|Modulation|DeactivateControlBusMix _cbm_teleport)
    (else
      _))
  (Utilities|Time|ClearandInvalidateTimerbyHandle (Variables|Default|GetCBM_Handle)))

(event Custom|EventOnDied
  (params (DamagedActor :type Actor%20Object%20Reference) (Damage :type Float%20%28single-precision%29) (InstigatedBy :type Controller%20Object%20Reference) (DamageCauser :type Actor%20Object%20Reference) (HitLocation :type Vector) (FHitComponent :type Primitive%20Component%20Object%20Reference) (BoneName :type Name) (Momentum :type Vector))
  (CallFunction|ClearOldMarkersAndProjectiles))

(fn Clearoldmarkersandprojectiles
  (bind _bgatoteleport (Variables|FortPawnComponent_VoidTeleport|GetBGAToTeleport))
  (Utilities|IsValid (Variables|FortPawnComponent_VoidTeleport|GetBGAToTeleport)
    (:Is%20Valid
      ;[0.4254 0.2346 0.5312 0.5] FX: Passing a parameter to tell the beacon FX to fade-out before being destroyed
      (bind _asbga_winter_decay_moving_teleport_marker (Utilities|Casting|CastToBGA_WinterDecay_MovingTeleportMarker _bgatoteleport)
        (:then
          (Effects|Components|ParticleSystem|SetBoolParameter (Variables|/WinterDecay/Gameplay/Actors/BGA_WinterDecay_MovingTeleportMarker.BGA_WinterDecay_MovingTeleportMarker_C|GetNS_WinterDecay_Mask_Portal _asbga_winter_decay_moving_teleport_marker) "Destroyed" true))
        (:CastFailed
          _))
      (Actor|DestroyActor _bgatoteleport))
    (:Is%20Not%20Valid
      _))
  (Utilities|IsValid (Variables|FortPawnComponent_VoidTeleport|GetCurrentTrackedVoidProjectile)
    (:Is%20Valid
      (Actor|DestroyActor (Variables|FortPawnComponent_VoidTeleport|GetCurrentTrackedVoidProjectile)))))

(fn GetTeleportPositionFromBGA
  (bind _bgatoteleport (Variables|FortPawnComponent_VoidTeleport|GetBGAToTeleport))
  (bind _returnvalue (|CallParentFunction))
  (Development|PrintString (Utilities|GetDisplayName _bgatoteleport) true true "(R=1.000000,G=0.290341,B=0.169553,A=1.000000)" 2e+01)
  (return (Math|Vector|SelectVector (Variables|Default|GetProjectileLocation) _returnvalue (Math|Boolean|ANDBoolean (Utilities|IsValid _bgatoteleport :!class /Script/Engine.KismetSystemLibrary) (ScalableFloat|GetValueasBool (Variables|Default|GetHF_UseLastProjectileLocationForTeleportLocation)) (not (Variables|Default|GetIsBGAOnTrain))))))

(fn HandleProjectileBeforeTeleport
  (bind _currenttrackedvoidprojectile (Variables|FortPawnComponent_VoidTeleport|GetCurrentTrackedVoidProjectile))
  (Utilities|IsValid _currenttrackedvoidprojectile
    (:Is%20Valid
      (Components|Activation|SetActive (Variables|/Script/FortniteGame.FortProjectileBase|GetProjectileMovementComponent _currenttrackedvoidprojectile) (not (ScalableFloat|GetValueasBool (Variables|Default|GetHF_StopProjectileOnTeleportPressed))) false))))

(fn IsAnyTeleporterMarkerAvailable
  (Utilities|IsValid (Variables|FortPawnComponent_VoidTeleport|GetBGAToTeleport)
    (:Is%20Valid
      (return true false))
    (:Is%20Not%20Valid
      (Utilities|IsValid (Variables|FortPawnComponent_VoidTeleport|GetCurrentTrackedVoidProjectile)
        (:Is%20Valid
          (return true true))
        (:Is%20Not%20Valid
          (return false false))))))

(fn NewBGACreated
  (params (BGAVoidMaskTeleportMarker :type Building%20Actor%20Object%20Reference))
  (bind _asbga_demon_mask_void_teleport_base (Utilities|Casting|CastToBGA_DemonMask_Void_Teleport_Base BGAVoidMaskTeleportMarker)
    (:then
      (Development|PrintString (Utilities|GetDisplayName BGAVoidMaskTeleportMarker) true true "(R=1.000000,G=0.295363,B=0.721485,A=1.000000)" 2e+01)
      (CallFunction|ServerSetBGAToTeleport _asbga_demon_mask_void_teleport_base)
      (CallFunction|SetCurrentMarkeronDemonMask _asbga_demon_mask_void_teleport_base)
      (CallFunction|TriggerTeleportMarkerCue)
      ;[0.3265 0 0.9948 0.5] Teleport carrying DBNO players
      (if (ScalableFloat|GetValueasBool (Variables|Default|GetHF_AllowOtherEquippedWeaponsDBNOTeleport))
        (CallFunction|TryTeleportWhileCarryingDBNOPlayerThread)))
    (:CastFailed
      (Development|PrintString "BGA ISN'T VALID"))))

(fn NewProjectileCreated
  (params (DemonMaskProjectile :type B%20Prj%20Winter%20Decay%20Object%20Reference))
  (CallFunction|ClearOldMarkersAndProjectiles)
  (Utilities|IsValid DemonMaskProjectile
    (:Is%20Valid
      (CallFunction|ServerSetCurrentTrackedVoidProjectile DemonMaskProjectile)
      (Default|BindEventtoOnBGACreated DemonMaskProjectile (EventDispatchers|CreateEvent :!event NewBGACreated)))
    (:Is%20Not%20Valid
      (Development|PrintString "ProjectileNotValid"))))

(fn NewWeaponEquipped
  (params (NewWeapon :type Fort%20Weapon%20Object%20Reference) (PrevWeapon :type Fort%20Weapon%20Object%20Reference))
  (Utilities|IsValid (Variables|Default|GetWinterDecayWeaponReference)
    (:Is%20Valid
      (Default|UnbindallEventsfromOnNewProjectileCreatedandValid (Variables|Default|GetWinterDecayWeaponReference)))
    (:Is%20Not%20Valid
      _))
  (bind _asb_winter_decay_teleport (Utilities|Casting|CastToB_WinterDecay_Teleport NewWeapon))
  (Utilities|IsValid _asb_winter_decay_teleport
    (:Is%20Valid
      (bind _output_get (Variables|Default|SetWinterDecayWeaponReference _asb_winter_decay_teleport))
      (Default|BindEventtoOnNewProjectileCreatedandValid _output_get (EventDispatchers|CreateEvent :!event NewProjectileCreated))
      ;[0.3265 0 0.9948 0.5] Teleport carrying DBNO players
      (if (ScalableFloat|GetValueasBool (Variables|Default|GetHF_AllowOtherEquippedWeaponsDBNOTeleport))
        (Ability|Async|EndAction (Variables|Default|GetCarrying%20Player%20Async%20Action))))))

(fn SetCurrentMarkerOnDemonMask
  (params (NewCurrentMarker :type BGA%20Demon%20Mask%20Void%20Teleport%20Base%20Object%20Reference))
  (Class|BWinterDecayTeleport|SetVoidTeleportMarker (Variables|Default|GetWinterDecayWeaponReference) NewCurrentMarker))

(fn VoidTeleportToLocation
  (params (LocationToTeleport :type Vector :default "0, 0, 0") (bIsJustTest :type Boolean) (bNoChecks :type Boolean))
  (bind _returnvalue (|CallParentFunction LocationToTeleport bIsJustTest bNoChecks))
  (CallFunction|PlayerComponentEndTeleport)
  (if _returnvalue
    (GameplayCue|ExecuteGameplayCue (Components|GetOwner) (Variables|Default|GetTeleport%20Ended%20Gameplay%20Cue%20Tag)))
  (return _returnvalue))
